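# Physics asset for the Skull Brite pickaxe, a two-handed melee weapon cosmetic with dangling rope and knot bones.
# Capsule bodies cover the long handle and five small dynamic bones: three ropes and two knots.
# The three ropes and the first knot each constrain directly to the handle, while the second knot chains off the first.
# Joints allow moderate swing with almost no twist.
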PhysicsAsset /BRCosmetics/Weapons/FORT_Melee/Pickaxe_SkullBrite/Meshes/Pickaxe_Skull_Brite_Physics
bodies {
 Melee_TwoHanded_Handle [{shapeType=Capsule, length=96.9976, radius=2.4661, center=<0.5613 -0.9356 15.0386>}]
 dyn_knot_01 [{shapeType=Capsule, length=1.3838, radius=1.6908, center=<1.3082 0 0.0001>, rotation=<rot 90 -1.2659 -1.266>}]
 dyn_knot_02 [{shapeType=Capsule, length=1.7108, radius=2.3879, center=<1.5609 -0.0955 -0.5667>, rotation=<rot 12.2821 90.0788 -86.6553>}]
 dyn_rope_01 [{shapeType=Capsule, length=3.4641, radius=1.4962, center=<3.4487 0.1075 0>, rotation=<rot -90 -0.0197 0.0195>}]
 dyn_rope_02 [{shapeType=Capsule, length=3.4291, radius=1.416, center=<3.4336 -0.167 -0.6668>, rotation=<rot -79.9918 179.9999 174.9999>}]
 dyn_rope_03 [{shapeType=Capsule, length=4.2216, radius=0.8238, center=<2.1064 -0.0175 0.14>, rotation=<rot 0 101.4488 -78.0561>}]
}
constraints [
 {bone1Name=dyn_rope_01, bone2Name=Melee_TwoHanded_Handle, swing1LimitDegrees=45, swing2LimitDegrees=60, twistLimitDegrees=1}
 {bone1Name=dyn_rope_02, bone2Name=Melee_TwoHanded_Handle, swing1LimitDegrees=45, swing2LimitDegrees=60, twistLimitDegrees=1}
 {bone1Name=dyn_rope_03, bone2Name=Melee_TwoHanded_Handle, swing1LimitDegrees=45, swing2LimitDegrees=45, twistLimitDegrees=1}
 {bone1Name=dyn_knot_01, bone2Name=Melee_TwoHanded_Handle, swing1LimitDegrees=45, swing2LimitDegrees=45, twistLimitDegrees=1}
 {bone1Name=dyn_knot_02, bone2Name=dyn_knot_01, swing1LimitDegrees=45, swing2LimitDegrees=45, twistLimitDegrees=1}
]
